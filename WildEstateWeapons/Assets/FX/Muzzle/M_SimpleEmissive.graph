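# Unlit translucent material for Niagara mesh particles and ribbons in the weapon muzzle FX set.
# An opacity texture channel, shaped by dynamic-parameter power, multiply, add and contrast, is multiplied by particle alpha and camera depth fade.
# A platform switch picks this value for Xb1/PS4, PC-Perf/Android and Switch, while other PC tiers and newer consoles get a standalone DepthFade without that mask.
# No emissive color logic appears.
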Material /WildEstateWeapons/Assets/FX/Muzzle/M_SimpleEmissive
graph {
 properties {
  bAutomaticallySetUsageInEditor=false
  bUsedWithNiagaraMeshParticles=true
  bUsedWithNiagaraRibbons=true
  blendMode=BLEND_Translucent
  ditherOpacityMask=true
  shadingModel=MSM_Unlit
  usedShadingModels=Unlit
 }
 nodes [
  n1 = DepthFade(FadeDistance=ScalarParameter() {parameterName=Depth fade})
  n2 = ChannelMaskParameter(None=TextureSampleParameter2D() {parameterName=Opacity Tex, texture=/Engine/EditorLandscapeResources/WhiteSquareTexture}.RGBA) {parameterName=Opacity Tex Channel}
  n3 = DynamicParameter() {defaultValue=<rgba 1 1 0 0>, paramNames [Opac Mask Pow, Opac Mask mul, Opac Mask add, Opac Mask contrast]}
  n4 = Add(A=Multiply(A=Power(Base=@n2.R, Exp=@n3.Opac Mask Pow), B=@n3.Opac Mask mul), B=@n3.Opac Mask add)
  n5 = Multiply(
   A=Multiply(
    A=ParticleColor().A
    B=MaterialFunctionCall(In=@n4, Contrast=@n3.Opac Mask contrast) {materialFunction=&CheapContrast}.Result
   )
   B=MaterialFunctionCall(
    Fade Length=ScalarParameter() {parameterName=Camera depth fade length}
    Fade Offset=ScalarParameter() {parameterName=Camera Depth fade Offset}
   ) {materialFunction=&CameraDepthFade}.Result
  )
  MaterialFunctionCall("Xb1,PS4"=@n5, "PC-High&&Epic,XSX,PS5"=@n1, "PC-Perf,Android"=@n5, Switch=@n5, "PC-Low(NotPerf)&&Medium"=@n1) {materialFunction=&PlatformSwitch_1D}
 ]
}
