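# Unlit translucent Niagara mesh-particle material for the outer layer of a toon muzzle flash, colored by particle color.
# Panned, skewed noise and vertex-color-selected front/back band shaping set a generated band's width for hard-edged opacity, with camera fade and platform-dependent depth fade.
# World position offset adds thickness and facing-dependent stretch, aligns the mesh to the camera, then rotates it about the view vector using an angle derived in clip space.
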Material /WildEstateWeapons/Assets/FX/Muzzle/M_ToonMuzzleFlashOuter
graph {
 properties {
  bAutomaticallySetUsageInEditor=false
  bUseMaterialAttributes=true
  bUsedWithNiagaraMeshParticles=true
  blendMode=BLEND_Translucent
  ditherOpacityMask=true
  shadingModel=MSM_Unlit
  usedShadingModels=Unlit
 }
 nodes [
  n1 = Constant3Vector() {constant=<rgba 10 0 0 0>}
  n2 = Subtract(
   A=MaterialFunctionCall(World Position=TransformPosition(None=@n1) {transformType=TRANSFORMPOSSOURCE_World}) {materialFunction=&TransformToClipSpace}.Clip Space XY
   B=MaterialFunctionCall(World Position=ObjectPositionWS()) {materialFunction=&TransformToClipSpace}.Clip Space XY
  )
  n3 = MaterialFunctionCall(Float2=Normalize(VectorInput=ComponentMask(None=@n2) {g=true, r=true})) {materialFunction=&BreakOutFloat2Components}
  Arctangent2!(Y=@n3.G, X=@n3.G)
  DynamicParameter() {parameterIndex=2, paramNames [NA,NA,NA,NA]}
  DynamicParameter() {parameterIndex=1, paramNames [NA,NA,NA,NA]}
  DynamicParameter() {parameterIndex=3, paramNames [NA,NA,NA,NA]}
  MaterialFunctionCall!(View Space Position, Buffer UV Pixel Offset) {materialFunction=/Game/Packages/Fortress_SharedMaterials/Base_Material_Functions/ViewSpaceTransformToClipSpace}
  n4 = TextureCoordinate()
  Multiply(A=@n4, B=Constant2Vector() {g=1, r=0.5})
  NamedRerouteDeclaration(None=DynamicParameter() {paramNames [Time,NA,NA,NA]}.Time) {name=DP0 R Time, nodeColor=<rgba 1 0 0 1>}
  Normalize(VectorInput=Transform(None=Constant3Vector() {constant=<rgba 1 0 0 0>}) {transformSourceType=TRANSFORMSOURCE_Local, transformType=TRANSFORM_Camera})
  n5 = Abs(None=Sine(None=MaterialFunctionCall(Float2=@n4) {materialFunction=&BreakOutFloat2Components}.R))
  Power(Base=@n5) {constExponent=1}
  n6 = ParticleColor()
  n7 = MaterialFunctionCall(Float2=TextureCoordinate()) {materialFunction=&BreakOutFloat2Components}
  n8 = VertexColor()
  n9 = Multiply(A=Subtract(A=Round(None=@n8.G)) {constB=0.5}) {constB=2}
  n10 = VectorParameter() {parameterName=Texture Options, channelNames {a="Pan Angle(0-1)", b=Skew amount, g=Tile Y, r=Tile X}, defaultValue=<rgba 10 0.35 0.2 0.5>}
  n11 = Multiply(A=@n7.G, B=Multiply(A=@n10.Skew amount, B=NamedRerouteUsage())) {constB=0.1}
  n12 = MaterialFunctionCall(X=Add(A=@n7.R, B=Multiply(A=@n9, B=@n11)), Y=@n7.G) {materialFunction=&MakeFloat2}
  n13 = AppendVector(A=@n10.Tile X, B=@n10.Tile X)
  n14 = Sine(None=@n10."Pan Angle(0-1)")
  n15 = Cosine(None=@n10."Pan Angle(0-1)")
  n16 = TextureSampleParameter2D!(
   UVs=Panner(Coordinate=Multiply(A=@n12.Result, B=@n13), Time=NamedRerouteUsage(), Speed=AppendVector(A=@n14, B=@n15)) {bFractionalPart=true, speedY=1}
   Bias=ScalarParameter() {parameterName=Spikey noise Tex Bias}
   Apply View MipBias
  ) {
   mipValueMode=TMVM_MipBias
   parameterName=spiky Noise Texture base
   samplerType=SAMPLERTYPE_LinearGrayscale
   texture=&/T_SphereHeight
  }
  n17 = ChannelMaskParameter(None=@n16.RGBA) {parameterName=Tex Channel}
  n18 = LinearInterpolate(A=@n17.R, B=OneMinus(None=@n17.R), Alpha=ScalarParameter() {defaultValue=1, parameterName=Invert Texture Channel})
  n19 = VectorParameter() {parameterName=Texture mask Options, channelNames {a=Contrast, b=add, g=mul, r=pow}, defaultValue=<rgba 2 1 0 0>}
  n20 = Add(A=Multiply(A=Power(Base=@n18, Exp=@n19.pow), B=@n19.mul), B=@n19.add)
  n21 = Multiply(A=MaterialFunctionCall(UV Channel=@n4) {materialFunction=&LinearGradient}.UGradient) {constB=4}
  n22 = OneMinus(None=@n21)
  n23 = MaterialFunctionCall!(Value=@n21) {materialFunction=/FortUILibrary/MaterialFunctions/Basics/MF_UI_EvenOrOdd}
  n24 = LinearInterpolate(
   A=Frac(None=LinearInterpolate(A=@n21, B=@n22, Alpha=@n23.Result))
   B=@n5
   Alpha=ScalarParameter() {defaultValue=1, parameterName=Bubble shaping}
  )
  n25 = VectorParameter() {parameterName=Band Options Back, channelNames {a=Contrast, b=add, g=mul, r=pow}, defaultValue=<rgba 1 1 0 0>}
  n26 = VectorParameter() {parameterName=Band Options Front, channelNames {a=Contrast, b=add, g=mul, r=pow}, defaultValue=<rgba 1 1 0.24 0>}
  n27 = VertexColor()
  n28 = MaterialFunctionCall(Float4=LinearInterpolate(A=AppendVector(A=@n25, B=@n25), B=AppendVector(A=@n26, B=@n26), Alpha=Round(None=@n27.R))) {materialFunction=&BreakOutFloat4Components}
  n29 = Add(A=Multiply(A=Power(Base=@n24, Exp=@n28.R), B=@n28.G), B=@n28.B)
  n30 = Multiply(A=Saturate(None=MaterialFunctionCall(In=@n29, Contrast=@n28.A) {materialFunction=&CheapContrast}.Result)) {constB=0.5}
  n31 = NamedRerouteUsage()
  n32 = Multiply(
   A=Multiply(
    A=MaterialFunctionCall(In=@n20, Contrast=@n19.Contrast) {materialFunction=&CheapContrast}.Result
    B=LinearInterpolate(B=@n30, Alpha=@n31) {constA=0.5}
   )
   B=ScalarParameter() {defaultValue=0.75, parameterName=Master Band Width}
  )
  n33 = ScalarParameter() {defaultValue=1, parameterName=Band Hardness}
  n34 = VectorParameter() {group=Opac, parameterName=Depth Fade Options, channelNames {a="", b=Camera Fade Offset, g=Camera Fade Length, r=Depth Fade Value}}
  n35 = Multiply(
   A=Multiply(
    A=@n6.A
    B=MaterialFunctionCall!(Width=@n32, Sharpness=@n33, Offset, Direction Switch, Compare, Input Coordinates) {materialFunction=/Engine/Functions/Engine_MaterialFunctions03/Procedurals/GeneratedBand}.Result
   )
   B=MaterialFunctionCall(Fade Length=@n34.Camera Fade Length, Fade Offset=@n34.Camera Fade Length) {materialFunction=&CameraDepthFade}.Result
  )
  n36 = DepthFade(Opacity=@n35, FadeDistance=@n34.Depth Fade Value)
  n37 = ScalarParameter() {parameterName=Thickness Offset back}
  n38 = LinearInterpolate(A=@n37, B=ScalarParameter() {defaultValue=5, parameterName=Thickness Offset Front}, Alpha=NamedRerouteUsage())
  n39 = VectorParameter() {parameterName=front back Thhickness Options, channelNames {a=Contrast, b=add, g=mul, r=pow}, defaultValue=<rgba 1 1 -0.3195 0.36>}
  n40 = Multiply(A=Power(Base=VertexColor().R, Exp=@n39.pow), B=@n39.mul)
  n41 = Saturate(None=MaterialFunctionCall(In=Add(A=@n40, B=@n39.add), Contrast=@n39.Contrast) {materialFunction=&CheapContrast}.Result)
  n42 = Multiply(A=Subtract(A=VertexColor().B) {constB=0.5}) {constB=2}
  n43 = Transform(None=CrossProduct(A=VertexNormalWS(), B=VertexTangentWS())) {transformSourceType=TRANSFORMSOURCE_World, transformType=TRANSFORM_Local}
  n44 = DotProduct(A=NamedRerouteUsage(), B=Transform(None=Constant3Vector() {constant=<rgba 1 0 0 0>}) {transformSourceType=TRANSFORMSOURCE_Instance})
  n45 = VectorParameter() {parameterName=front back Dot Options, channelNames {a=Contrast, b=add, g=mul, r=pow}, defaultValue=<rgba 1 1 0 0>}
  n46 = Power(Base=OneMinus(None=Abs(None=@n44)), Exp=@n45.pow)
  n47 = MaterialFunctionCall(In=Add(A=Multiply(A=@n46, B=@n45.mul), B=@n45.add), Contrast=@n45.Contrast) {materialFunction=&CheapContrast}
  n48 = VectorParameter() {parameterName=WPO X Axis Shaping Options, channelNames {a=contrast, b=add, g=Mul, r=pow}, defaultValue=<rgba 2 1 -0.25 0>}
  n49 = Multiply(A=Power(Base=VertexColor().R, Exp=@n48.pow), B=@n48.Mul) {constB=1.5}
  n50 = Saturate(None=MaterialFunctionCall(In=Add(A=@n49, B=@n48.add) {constB=0}, Contrast=@n48.contrast) {materialFunction=&CheapContrast}.Result)
  n51 = Multiply(A=ScalarParameter() {defaultValue=10, parameterName=Streach Str}) {constB=-1}
  n52 = MaterialFunctionCall(Y=Multiply(A=NamedRerouteDeclaration(None=@n47.Result) {name=FrontBack Dot, nodeColor=<rgba 1 0.4922 0 1>}, B=Multiply(A=@n50, B=@n51))) {materialFunction=&MakeFloat3}
  n53 = Add(A=Multiply(A=LinearInterpolate(A=@n37, B=@n38, Alpha=@n41), B=Multiply(A=@n42, B=@n43)), B=@n52.Result)
  n54 = WorldPosition()
  n55 = MaterialFunctionCall!(
   WorldPosition=Add(A=Transform(None=@n53) {transformSourceType=TRANSFORMSOURCE_Instance}, B=@n54.XYZ)
   PivotLocation
   WS Normals
   Custom Object Basis 1=Transform(None=Constant3Vector() {constant=<rgba 0 0 1 0>}) {transformSourceType=TRANSFORMSOURCE_Instance}
   Custom Object Basis 2=Transform(None=Constant3Vector() {constant=<rgba 0 1 0 0>}) {transformSourceType=TRANSFORMSOURCE_Instance}
   Custom Object Basis 3=Transform(None=Constant3Vector() {constant=<rgba 1 0 0 0>}) {transformSourceType=TRANSFORMSOURCE_Instance}
  ) {materialFunction=/Engine/Functions/Engine_MaterialFunctions02/WorldPositionOffset/AlignMeshToTheCamera}
  n56 = CameraPositionWS()
  n57 = ObjectPositionWS()
  n58 = NamedRerouteDeclaration(None=Normalize(VectorInput=Subtract(A=@n56, B=@n57))) {name=Vector to Camera WS, nodeColor=<rgba 0.7812 1 0 1>}
  n59 = Arctangent2Fast(Y=@n3.G, X=@n3.G)
  n60 = Constant() {r=2}
  n61 = Add(A=Divide(
    A=Multiply(A=@n59) {constB=-1}
    B=MaterialFunctionCall(Multiplier=@n60) {materialFunction=&Pi}.Result
   )) {constB=0.25}
  n62 = ObjectPositionWS()
  n63 = Add(A=@n55.World Position Offset, B=@n54.XYZ)
  SetMaterialAttributes!(
   MaterialAttributes
   Emissive Color=@n6.RGB
   Opacity=MaterialFunctionCall("Xb1,PS4"=@n36, "PC-High&&Epic,XSX,PS5"=@n36, "PC-Perf,Android"=@n35, Switch=@n35, "PC-Low(NotPerf)&&Medium"=@n36) {materialFunction=&PlatformSwitch_1D}.Result
   Opacity Mask=@n35
   World Position Offset=Add(A=@n55.World Position Offset, B=RotateAboutAxis(NormalizedRotationAxis=@n58, RotationAngle=@n61, PivotPoint=@n62, Position=@n63))
  ) {attributeSetTypes [
    #1
    #2
    #3
    #4
   ]}
  TextureCoordinate()
  Transform(None=@n1)
  Transform(None=NamedRerouteUsage()) {transformSourceType=TRANSFORMSOURCE_World, transformType=TRANSFORM_Camera}
  Transform()
  Transform(None=@n52.Result) {transformSourceType=TRANSFORMSOURCE_World, transformType=TRANSFORM_Local}
 ]
}
